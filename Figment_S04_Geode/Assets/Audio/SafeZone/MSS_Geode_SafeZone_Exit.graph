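# MetaSound Source that plays a stereo one-shot for exiting a Geode safe zone, using two POI boundary exit waves.
# It passes the wave array and play trigger to a MetaSoundWizard array one-shot graph, which supplies both audio channels and the finished trigger.
# The Master Volume input is left unconnected.
# Very-close attenuation applies, and owner-limited concurrency allows one instance, blocking new plays with a retrigger time.
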MetaSoundSource /Figment_S04_Geode/Assets/Audio/SafeZone/MSS_Geode_SafeZone_Exit
inputs [
 Master Volume {type=Float}
 Sounds {type=WaveAsset:Array}
 UE.Source.OnPlay {type=Trigger}
 UE.Source.StartTime.StartTime {type=Time}
]
interfaces [UE.OutputFormat.Stereo:1.0, UE.Source.OneShot:1.0, UE.Source.StartTime:1.0, UE.Source:1.1]
outputs [
 "UE.OutputFormat.Stereo.Audio:0" {type=Audio}
 "UE.OutputFormat.Stereo.Audio:1" {type=Audio}
 UE.Source.OneShot.OnFinished {type=Trigger}
]
pages [Default {
  input_defaults {
   Master Volume=1
   UE.Source.OnPlay=null
   UE.Source.StartTime.StartTime=null
   Sounds [/Figment_S04_Geode/Assets/Audio/wav/Geode_POI_Boundary_Exit_01, /Figment_S04_Geode/Assets/Audio/wav/Geode_POI_Boundary_Exit_02]
  }
  nodes [MSS_Oneshot_Array_Stereo_0 = /MetaSoundWizard/MetaSounds/One-shots/Array/MSS_Oneshot_Array_Stereo(Sounds=^Sounds, UE.Source.OnPlay=^UE.Source.OnPlay)]
  links [
   {dst="@UE.OutputFormat.Stereo.Audio:0", src=MSS_Oneshot_Array_Stereo_0.UE.OutputFormat.Stereo.Audio:0}
   {dst="@UE.OutputFormat.Stereo.Audio:1", src=MSS_Oneshot_Array_Stereo_0.UE.OutputFormat.Stereo.Audio:1}
   {dst="@UE.Source.OneShot.OnFinished", src=MSS_Oneshot_Array_Stereo_0.UE.Source.OneShot.OnFinished}
  ]
 }]
properties {
 assetImportData=&self:AssetImportData
 attenuationSettings=/Figment_S04_Geode/Assets/Settings/ATT_Geode_POI_VeryClose
 bOverrideConcurrency=true
 maxDistance=4100
 numChannels=2
 outputFormat=Stereo
 sampleRate=48000
 soundClassObject=&Fort_Events_Gameplay
 concurrencyOverrides {bLimitToOwner=true, maxCount=1, resolutionRule=PreventNew, retriggerTime=2}
}
